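; Gameplay ability for the HighTower Tomato Repulsor Cannon, a charged beam attack that slows the hovering player and can launch them backward when it ends.
; Activation replicates aim rotation and adds a charge-up cue before a montage notify limits turn input and starts server-side beam cues and timed targeting traces.
; A saved-hit array limits enemies to one hit per burst while structures take repeated hits, and the final trace schedules ability end.
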(event EventCastMontageBlendOut)

(event EventCastMontageCancelledOrInterrupted
  (Ability|EndAbility))

(event EventOnAbilityActivated
  ;[1 1 1] On Activated init the aim scale
  (|CallParentFunction)
  (Variables|Default|SetAimScale 1.0)
  (if (Networking|IsServer)
    ;[1 1 1] Start replicating control rotation on player pawn athena
    (bind _asplayer_pawn_athena (Utilities|Casting|CastToPlayerPawn_Athena (Variables|Default|GetPlayerPawn)))
    (Class|PlayerPawnAthena|StartAImRotationTick _asplayer_pawn_athena))
  ;[1 1 1] This gives the control rotation a moment to replicate before the charge up fx starts.
  (Utilities|FlowControl|Delay 0.1)
  (Ability|AddGameplayCueToOwner (Variables|Default|GetGC_ChargeUp)))

(event Custom|HandleAimScale)

(event Custom|Trace
  ;[0.3582 1 0.4198] Targeting Trace. Trace happens multiple times, but targeted enemies will only be damaged once per burst. Structures on the other hand will be hit multiple times.
  (CallFunction|PlayBurstGC (Variables|PowerGameplayCues|GetGC_Targeting) (Variables|PowerGameplayCues|GetTargetingGCUsesParams) (CallFunction|TargetingGCParams))
  (Utilities|FlowControl|Sequence
    (:then_0
      (bind (_asynctaskproxy _targetdata _applicationtag) (Ability|Tasks|DoTargetSelection "Targeting" "(TagName=\"Targeting.ApplicationTag.0\")" true)
        (:Targeted
          (Variables|Default|SetTargetDataArray (Ability|TargetData|GetAllActorsfromTargetData _targetdata))
          (for _array_element (Variables|Default|GetSavedHitActors)
            (Utilities|Array|RemoveItem (Variables|Default|GetTargetDataArray) _array_element))
          (Development|PrintString (Utilities|GetDisplayName (Utilities|Array|Get(acopy) (Variables|Default|GetTargetDataArray) (Utilities|Array|LastIndex (Variables|Default|GetTargetDataArray)))))
          (CallFunction|ApplyEffect (Ability|TargetData|AbilityTargetDatafromActorArray (Variables|Default|GetTargetDataArray)) _applicationtag)
          ;;[1 1 1] Added Actors that were hit this Trace to the "Saved Hit Actors" Array.
          (Utilities|Array|AppendArray (Variables|Default|GetSavedHitActors) (Variables|Default|GetTargetDataArray)))))
    (:then_1
      ;;[1 1 1] StructureTargeting
      (bind (_asynctaskproxy_2 _targetdata_2 _applicationtag_2) (Ability|Tasks|DoTargetSelection "Targeting" "(TagName=\"Targeting.ApplicationTag.1\")" true)
        (:Targeted
          (CallFunction|ApplyEffect _targetdata_2 _applicationtag_2))))))

(event Custom|TargetTick
  ;[1 1 1] Trace Timer Loop
  (bind _counter (Utilities|FlowControl|DoN (Math|Float|Truncate (ScalableFloat|GetValueatLevel (Variables|Default|GetNumberOfTraces))))
    (:Exit
      (CallFunction|Trace)
      (if (>= _counter (- (Math|Float|Truncate (ScalableFloat|GetValueatLevel (Variables|Default|GetNumberOfTraces))) 1))
        (Utilities|Time|SetTimerbyEvent (delegate Custom|CustomEvent_0) 0.25)))))

(event Custom|CustomEvent_0
  ;[1 1 1] Trace Timer Loop
  (Ability|EndAbility))

(event EventCastMontageNotifyStart
  (params (NotifyTag :type Gameplay%20Tag%20Structure))
  (bind _playerpawn (Variables|Default|GetPlayerPawn))
  (bind _returnvalue (Pawn|GetControlRotation _playerpawn :!class /Script/Engine.Pawn))
  (bind _returnvalue_2 (Transformation|GetActorLocation _playerpawn))
  (bind _returnvalue_3 (Math|Vector|GetForwardVector _returnvalue))
  (bind _range (Variables|Default|GetRange))
  (bind _returnvalue_4 (ScalableFloat|GetValueatLevel _range))
  (bind _returnvalue_5 (* _returnvalue_3 _returnvalue_4))
  (bind _returnvalue_6 (+ _returnvalue_2 _returnvalue_5))
  ;[1 1 1] Check what notify we hit in the montage and trigger the ability based on it.
  (Utilities|FlowControl|Sequence
    (:then_0
      ;[1 1 1] Trace Timer Loop
      (bind _counter (Utilities|FlowControl|DoN (Math|Float|Truncate (ScalableFloat|GetValueatLevel (Variables|Default|GetNumberOfTraces))))
        (:Exit
          (CallFunction|Trace)
          (if (>= _counter (- (Math|Float|Truncate (ScalableFloat|GetValueatLevel (Variables|Default|GetNumberOfTraces))) 1))
            (Utilities|Time|SetTimerbyEvent (delegate Custom|CustomEvent_0) 0.25)))))
    (:then_1
      (if (GameplayTags|Equal(GameplayTag) NotifyTag (Variables|PowerTargetApply|GetTag_TargetApplyNotify))
        (Utilities|DoOnceThenDo false
          (:DoOnce
            _)
          (:ThenDo
            ;[1 1 1] Handle Commit and do targeting
            (if (Variables|PowerAbilityCommit|GetCommitOnEveryTargeting)
              (Utilities|PowerCommitAbility (Variables|AbilityCommit|GetIgnoreAbilityCommit) :!label _label
                (:Pass
                  _)
                (:Fail
                  (Development|PrintString "Power: Ability Commit Failed")
                  (CallFunction|PowerCommitFailed))
                (:Ignored
                  _)))
            (CallFunction|UpdateRotationInputScale true)
            (Utilities|IsServer
              (:true
                ;[1 1 1] Clear Saved Hit Actors
                (Utilities|Array|Clear (Variables|Default|GetSavedHitActors))
                ;[1 1 1] START TARGETING
                (CallFunction|Trace)
                (bind _returnvalue_7 (Utilities|Time|SetTimerbyEvent (delegate Custom|TargetTick) (ScalableFloat|GetValueatLevel (Variables|Default|GetTimeBetweenTraces)) true))
                ;[1 1 1] Add looping Beam cue
                (Ability|AddGameplayCueToOwner (Variables|Default|GetGC_BeamLoop))
                ;[1 1 1] Trace for Beam End Point
                (bind (_outhit _returnvalue_8) (Collision|LineTraceByProfile _returnvalue_2 _returnvalue_6 "FortTrajectoryBlocksWater" true (Utilities|Array|MakeArray (Ability|GetAvatarActorfromActorInfo))))
                ;[1 0.7965 0.2705] Hit World GC.
                (Ability|ExecuteGameplayCueWithParamsOnOwner (Variables|Default|GetHitWorldGC) (Ability|GameplayCue|MakeGameplayCueParameters :NormalizedMagnitude 0.0 :RawMagnitude 0.0 :Location (Math|Vector|SelectVector (.Location (Collision|BreakHitResult _outhit)) _returnvalue_6 _returnvalue_8) :Normal (.ImpactNormal (Collision|BreakHitResult _outhit)) :Instigator (.HitActor (Collision|BreakHitResult _outhit)) :GameplayEffectLevel 1 :AbilityLevel 1 :bReplicateLocationWhenUsingMinimalRepProxy false))))))
        (goto _label)
        (elif (and (GameplayTags|Equal(GameplayTag) NotifyTag (Variables|PowerAnimationCast|GetT_AbilityEndCastNotify)) (GameplayTags|NotEqual(GameplayTag) :A (Variables|PowerAnimationCast|GetT_AbilityEndCastNotify)))
          (CallFunction|CastMontageNotifyEndAbility))))
    (:then_2)))

(event Custom|Reset
  (bind _playerpawn (Variables|Default|GetPlayerPawn))
  (bind _returnvalue (Pawn|GetControlRotation _playerpawn :!class /Script/Engine.Pawn))
  (bind _returnvalue_2 (Transformation|GetActorLocation _playerpawn))
  (bind _returnvalue_3 (Math|Vector|GetForwardVector _returnvalue))
  (bind _range (Variables|Default|GetRange))
  (bind _returnvalue_4 (ScalableFloat|GetValueatLevel _range))
  (bind _returnvalue_5 (* _returnvalue_3 _returnvalue_4))
  (bind _returnvalue_6 (+ _returnvalue_2 _returnvalue_5))
  ;[1 1 1] Check what notify we hit in the montage and trigger the ability based on it.
  (Utilities|DoOnceThenDo false
    (:DoOnce
      _)
    (:ThenDo
      ;[1 1 1] Handle Commit and do targeting
      (if (Variables|PowerAbilityCommit|GetCommitOnEveryTargeting)
        (Utilities|PowerCommitAbility (Variables|AbilityCommit|GetIgnoreAbilityCommit) :!label _label
          (:Pass
            _)
          (:Fail
            (Development|PrintString "Power: Ability Commit Failed")
            (CallFunction|PowerCommitFailed))
          (:Ignored
            _)))
      (CallFunction|UpdateRotationInputScale true)
      (Utilities|IsServer
        (:true
          ;[1 1 1] Clear Saved Hit Actors
          (Utilities|Array|Clear (Variables|Default|GetSavedHitActors))
          ;[1 1 1] START TARGETING
          (CallFunction|Trace)
          (bind _returnvalue_7 (Utilities|Time|SetTimerbyEvent (delegate Custom|TargetTick) (ScalableFloat|GetValueatLevel (Variables|Default|GetTimeBetweenTraces)) true))
          ;[1 1 1] Add looping Beam cue
          (Ability|AddGameplayCueToOwner (Variables|Default|GetGC_BeamLoop))
          ;[1 1 1] Trace for Beam End Point
          (bind (_outhit _returnvalue_8) (Collision|LineTraceByProfile _returnvalue_2 _returnvalue_6 "FortTrajectoryBlocksWater" true (Utilities|Array|MakeArray (Ability|GetAvatarActorfromActorInfo))))
          ;[1 0.7965 0.2705] Hit World GC.
          (Ability|ExecuteGameplayCueWithParamsOnOwner (Variables|Default|GetHitWorldGC) (Ability|GameplayCue|MakeGameplayCueParameters :NormalizedMagnitude 0.0 :RawMagnitude 0.0 :Location (Math|Vector|SelectVector (.Location (Collision|BreakHitResult _outhit)) _returnvalue_6 _returnvalue_8) :Normal (.ImpactNormal (Collision|BreakHitResult _outhit)) :Instigator (.HitActor (Collision|BreakHitResult _outhit)) :GameplayEffectLevel 1 :AbilityLevel 1 :bReplicateLocationWhenUsingMinimalRepProxy false))))))
  (goto _label))

(event EventSetHover
  (params (On :type Boolean))
  ;[1 1 1] Slow During Hover
  (|CallParentFunction On)
  (if On
    (if (Variables|Default|GetSlowDuringHover)
      (bind _returnvalue (Ability|ApplyGameplayEffectToOwner "/HighTower/Items/Tomato/RepulsorCannon/GE_HighTower_Tomato_Slow.GE_HighTower_Tomato_Slow_C"))
      (Variables|Default|SetGE_Slow_Handle _returnvalue))))

(event EventCastMontageNotifyEndAbility
  ;[1 0.3784 0.3086] Ability End
  (Development|PrintString "RepulsorEnd"))

(event Ability|EventOnEndAbility
  (params (bWasCancelled :type Boolean))
  ;[1 0.3784 0.3086] Ability End
  (Utilities|Time|ClearandInvalidateTimerbyHandle (Variables|Default|GetLimitRotationTickHandle))
  (Ability|RemoveGameplayCueFromOwner (Variables|Default|GetGC_BeamLoop))
  (Utilities|Time|ClearandInvalidateTimerbyHandle (stale EventCastMontageNotifyStart _returnvalue_7))
  (bind _asplayer_pawn_athena (Utilities|Casting|CastToPlayerPawn_Athena (Variables|Default|GetPlayerPawn))
    (:then
      (Class|PlayerPawnAthena|EndRotationTick _asplayer_pawn_athena))
    (:CastFailed
      _))
  ;[1 1 1] Remove Slow
  (Ability|RemoveGameplayEffectFromOwnerWithHandle (Variables|Default|GetGE_Slow_Handle))
  ;[1 1 1] Launch Player
  (if (Variables|Default|GetDoLaunchOnAbilityEnd)
    (CallFunction|LaunchPlayer (Ability|GetAvatarActorfromActorInfo)))
  (CallFunction|SetHover false)
  (CallFunction|UpdateRotationInputScale false)
  ;[1 1 1] Call Parent end
  (|CallParentFunction bWasCancelled))

(fn LaunchPlayer
  (params (Actor :type Actor%20Object%20Reference))
  (bind _asfort_player_pawn (Utilities|Casting|CastToFortPlayerPawn Actor))
  (Character|LaunchCharacter _asfort_player_pawn (+ (* (* (Math|Vector|GetForwardVector (Pawn|GetControlRotation (Fortnite|GetFortPlayerControllerfromActor _asfort_player_pawn) :!class /Script/Engine.Controller)) -1.0) (Variables|Default|GetImpulseStrength)) (Math|Vector|MakeVector 0.0 0.0 (Variables|Default|GetAdditionalZLaunch))) true true))

(fn TargetingGCParams
  (return (Ability|GameplayCue|MakeGameplayCueParameters :NormalizedMagnitude 0.0 :RawMagnitude 0.0 :Location (Variables|Default|GetFxBeamEndPoint) :Normal "0,0,0" :GameplayEffectLevel 1 :AbilityLevel 1 :bReplicateLocationWhenUsingMinimalRepProxy false)))

(fn UpdateRotationInputScale
  (params (LimitRotation :type Boolean) (InInputPitchScale :type Float%20%28double-precision%29) (InInputYawScale :type Float%20%28double-precision%29))
  (bind _returnvalue (Ability|GetAvatarActorfromActorInfo))
  (bind _instigator (Variables|/Script/Engine.Actor|GetInstigator _returnvalue))
  (bind _returnvalue_2 (Pawn|GetController _instigator))
  (Utilities|IsValid _returnvalue_2
    (:Is%20Valid
      (bind _asfort_player_controller (Utilities|Casting|CastToFortPlayerController _returnvalue_2))
      (if LimitRotation
        (Input|SetInputPitchScale _asfort_player_controller InInputPitchScale)
        (Input|SetInputYawScale _asfort_player_controller InInputYawScale)
        (else
          (Input|ResetInputRotationScale _asfort_player_controller))))))
